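; Flintlock Pistol gameplay ability that knocks back the firing player and grants fall damage immunity.
; On commit execute, it checks the avatar actor against a set of knockback-prevention gameplay tags.
; If none match, it applies the self-knock and fall-damage-immunity effects to the owner, then calls the parent implementation.
; The parent call sits inside that branch, so a matching tag skips it along with both effects.
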(event Ability|EventCommitExecute
  ;[1 1 1] Apply Knock Effect and Fall Damage Immunity to self when firing
  (bind _returnvalue (GameplayTags|HasAnyMatchingGameplayTags(Message) (Ability|GetAvatarActorfromActorInfo) (Variables|Default|GetTags_PreventKnockback)))
  (if (not _returnvalue)
    (Ability|ApplyGameplayEffectToOwner (Variables|Default|GetGE_SelfKnock))
    (Ability|ApplyGameplayEffectToOwner (Variables|Default|GetGE_FallDamageImmunity))
    (|CallParentFunction)))
